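# Sound cue for the stop sting of the Apollo stage DJ booth prop's music, using a "Freestylin In Da Club" stop sound.
# A single wave player feeds the output directly, with no randomization or modulation.
# It references the booth's shared attenuation asset, routes through a quest music sound class, and plays slightly below full volume.
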SoundCue /Game/Sounds/Apollo/Props/DonutDJBooth/PGO_Stage_DJ_Booth_MusicStop_Cue
root=WavePlayer_1
nodes [{id=WavePlayer_1, properties {soundWaveAssetPtr=&/S12_Freestylin_In_Da_Club_Stop}}]
properties {
 attenuationSettings=&/PGO_Stage_DJ_Booth_Attenuation_Cue
 maxDistance=6144
 soundClassObject=/Game/Sounds/FortSoundClassesAndModes/Fort_Music_Quest_PSM
 volumeMultiplier=0.9
 attenuationOverrides {bAttenuate=false}
}
